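# Physics asset for the Prison Break backpack cosmetic (F_MED), defining capsule bodies and joint limits for its dangling parts.
# The pack body anchors constrained chains: three-link left and right flaps, a two-link strap, a two-link right pocket, plus single-link buckle and zip.
# Flaps and pocket bend mostly on one swing axis like hinges, with the right flap slightly looser, while strap and zip swing more freely.
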PhysicsAsset /BRCosmetics/Accessories/FORT_Backpacks/Backpack_F_MED_PrisonBreak/Meshes/F_MED_Prison_Break_Pack_Physics
bodies {
 dyn_buckle [{
   shapeType=Capsule
   length=3.4186
   radius=2.1711
   shapeName=dyn_buckle_capsule
   center=<2.2293 0.7109 -0.3905>
   rotation=<rot 0 -53.0523 -81.3017>
  }]
 dyn_flap_01_l [{shapeType=Capsule, length=5, radius=5, shapeName=dyn_flap_01_l_capsule, center=<-0.111 0 5.6053>, rotation=<rot 0 0 90>}]
 dyn_flap_01_r [{shapeType=Capsule, length=5, radius=5, shapeName=dyn_flap_01_r_capsule, center=<0 0.6831 -5.6053>, rotation=<rot 0 0 90>}]
 dyn_flap_02_l [{shapeType=Capsule, length=4, radius=4, shapeName=dyn_flap_02_l_capsule, center=<6.0606 0.5713 2.439>, rotation=<rot 0 0 90>}]
 dyn_flap_02_r [{shapeType=Capsule, length=4, radius=4, shapeName=dyn_flap_02_r_capsule, center=<-6.0003 -0.4785 -1.9902>, rotation=<rot 0 0 89.7413>}]
 dyn_flap_l [{shapeType=Capsule, length=4, radius=4, shapeName=dyn_flap_l_capsule, center=<-6.404 0 -0.3723>, rotation=<rot 0 0 -89.3326>}]
 dyn_flap_r [{shapeType=Capsule, length=4, radius=4, shapeName=dyn_flap_r_capsule, center=<6.404 0 -1.1947>, rotation=<rot 0 0 90>}]
 dyn_pocket_01_r [{
   shapeType=Capsule
   length=4.0183
   radius=2.0054
   shapeName=dyn_pocket_01_r_capsule
   center=<-0.595 0.0372 -1.4431>
   rotation=<rot 0 0.6086 -86.3791>
  }]
 dyn_pocket_r [{
   shapeType=Capsule
   length=0.505
   radius=0.505
   shapeName=dyn_pocket_r_capsule
   center=<2.046 -0.0288 -0.2186>
   rotation=<rot 0 -4.4065 -89.1078>
  }]
 dyn_strap [{shapeType=Capsule, length=0.505, radius=0.505, shapeName=dyn_strap_capsule, rotation=<rot 0 -90 0>}]
 dyn_strap_01 [{
   shapeType=Capsule
   length=2.0606
   radius=1.6315
   shapeName=dyn_strap_01_capsule
   center=<1.3507 -0.2444 -2.0828>
   rotation=<rot 0 -179.4665 -89.0142>
  }]
 dyn_zip [{shapeType=Capsule, length=1.85, radius=1.35, shapeName=dyn_zip_capsule, center=<1.5451 -0.014 0.7534>, rotation=<rot 0 95.2257 -86.0341>}]
 pack [{shapeType=Capsule, length=19.8787, radius=11.1982, shapeName=pack_capsule, center=<-1.4493 -6.1559 -5.1485>}]
}
constraints [
 {bone1Name=dyn_zip, bone2Name=pack, swing1LimitDegrees=45, swing2LimitDegrees=45, twistLimitDegrees=15}
 {bone1Name=dyn_strap_01, bone2Name=dyn_strap, swing1LimitDegrees=45, swing2LimitDegrees=45, twistLimitDegrees=1}
 {bone1Name=dyn_strap, bone2Name=pack, swing1LimitDegrees=35, swing2LimitDegrees=35, twistLimitDegrees=25}
 {bone1Name=dyn_pocket_01_r, bone2Name=dyn_pocket_r, swing1LimitDegrees=1, swing2LimitDegrees=10, twistLimitDegrees=1}
 {bone1Name=dyn_pocket_r, bone2Name=pack, swing1LimitDegrees=1, swing2LimitDegrees=10, twistLimitDegrees=1}
 {bone1Name=dyn_buckle, bone2Name=pack, swing1LimitDegrees=10, swing2LimitDegrees=10, twistLimitDegrees=1}
 {bone1Name=dyn_flap_02_r, bone2Name=dyn_flap_01_r, swing1LimitDegrees=1, swing2LimitDegrees=25, twistLimitDegrees=1}
 {bone1Name=dyn_flap_01_r, bone2Name=dyn_flap_r, swing1LimitDegrees=1, swing2LimitDegrees=45, twistLimitDegrees=1}
 {bone1Name=dyn_flap_r, bone2Name=pack, swing1LimitDegrees=1, swing2LimitDegrees=25, twistLimitDegrees=11}
 {bone1Name=dyn_flap_02_l, bone2Name=dyn_flap_01_l, swing1LimitDegrees=1, swing2LimitDegrees=25, twistLimitDegrees=1}
 {bone1Name=dyn_flap_01_l, bone2Name=dyn_flap_l, swing1LimitDegrees=1, swing2LimitDegrees=25, twistLimitDegrees=1}
 {bone1Name=dyn_flap_l, bone2Name=pack, swing1LimitDegrees=1, swing2LimitDegrees=25, twistLimitDegrees=1}
]
