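# Anim Blueprint for the Prison Break backpack cosmetic, adding physics-driven secondary motion and pose fixes for skydiving and jumping.
# Two RigidBody simulations, using default and skydive physics assets, run on the input pose and are blended by bIsSkyDiving.
# Additive bone rotations then adjust the pack and flaps, eased in and out through interpolated SkyFix and Jumpfix alphas.
# The EventGraph updates state-based gravity via a deprecated helper and sets both alphas, delaying jump changes.
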AnimBlueprint /BRCosmetics/Accessories/FORT_Backpacks/Backpack_F_MED_PrisonBreak/Meshes/F_MED_Prison_Break_Pack_AnimBP
target_skeleton=&/F_MED_Prison_Break_Pack_Skeleton
graphs [
 AnimGraph {nodes [
   n1 = Get GravityOverride()
   n2 = RigidBody!(ComponentPose=Local To Component(LocalPose=Use cached pose 'inputNode'()), Alpha, OverrideWorldGravity:Vector=@n1, ExternalForce) -> (Pose) {
    bOverrideWorldGravity=true
    overridePhysicsAsset=&/F_MED_Prison_Break_Pack_Physics
    alphaBoolBlend {blendOption=Linear}
    alphaScaleBiasClamp {bInterpResult=true, interpSpeedDecreasing=2, interpSpeedIncreasing=1}
    baseBoneRef {boneName=root}
    componentAppliedLinearAccClamp=<1000 1000 1000>
    componentLinearAccScale=<0.15 0.2>
    componentLinearVelScale=<0.22 0.212 0.22>
    overrideWorldGravity {z=-980}
    simSpaceSettings {
     maxAngularAcceleration=20
     maxAngularVelocity=1
     maxLinearAcceleration=300
     maxLinearVelocity=150
     velocityScaleZ=0
     worldAlpha=1
     externalLinearDragV=<0.2 0.2>
    }
   }
   n3 = Component To Local(ComponentPose=@n2)
   n4 = RigidBody(ComponentPose=Local To Component(LocalPose=Use cached pose 'inputNode'()), OverrideWorldGravity=@n1) {
    bOverrideWorldGravity=true
    overridePhysicsAsset=&/F_MED_Prison_Break_Pack_Physics_skydive
    alphaBoolBlend {blendOption=Linear}
    alphaScaleBiasClamp {bInterpResult=true, interpSpeedDecreasing=2, interpSpeedIncreasing=1}
    baseBoneRef {boneName=root}
    componentAppliedLinearAccClamp=<1000 1000 1000>
    componentLinearAccScale=<0.2 0.2>
    componentLinearVelScale=<0.22 0.212 0.22>
    overrideWorldGravity {z=-980}
    simSpaceSettings {
     maxAngularAcceleration=20
     maxAngularVelocity=1
     maxLinearAcceleration=300
     maxLinearVelocity=150
     velocityScaleZ=0
     worldAlpha=1
     externalLinearDragV=<0.2 0.2>
    }
   }
   n5 = Component To Local(ComponentPose=@n4)
   n6 = Get bIsSkyDiving()
   n7 = Get SkyFix() -> (SkyFix:double)
   n8 = "Transform (Modify) Bone - Bone: pack"(
    ComponentPose=Local To Component(LocalPose=Blend Poses by bool(BlendPose_0=@n5, BlendPose_1=@n3, bActiveValue=@n6) {blendType=Linear})
    Alpha=@n7
    Translation="0,0,0"
    Rotation="-0.000,0.000,23.800"
   ) {
    rotationMode=BMM_Additive
    rotationSpace=BCS_BoneSpace
    translationMode=BMM_Additive
    translationSpace=BCS_BoneSpace
    alphaBoolBlend {blendOption=Linear}
    alphaScaleBiasClamp {bInterpResult=true, interpSpeedDecreasing=1.7, interpSpeedIncreasing=4}
    boneToModify {boneName=pack}
    rotation {roll=23.8}
   }
   n9 = "Transform (Modify) Bone - Bone: base_flap_r"(ComponentPose=@n8, Alpha=@n7, Translation="0,0,0", Rotation="-10,-20,0", Scale="1,1,1") {
    rotationMode=BMM_Additive
    rotationSpace=BCS_BoneSpace
    translationMode=BMM_Additive
    translationSpace=BCS_BoneSpace
    alphaBoolBlend {blendOption=Linear}
    alphaScaleBiasClamp {bInterpResult=true, interpSpeedDecreasing=1.7, interpSpeedIncreasing=4}
    boneToModify {boneName=base_flap_r}
    rotation {pitch=-10, yaw=-20}
   }
   n10 = "Transform (Modify) Bone - Bone: base_flap_l"(ComponentPose=@n9, Alpha=@n7, Translation="0,0,0", Rotation="-10,-20,0", Scale="1,1,1") {
    rotationMode=BMM_Additive
    rotationSpace=BCS_BoneSpace
    translationMode=BMM_Additive
    translationSpace=BCS_BoneSpace
    alphaBoolBlend {blendOption=Linear}
    alphaScaleBiasClamp {bInterpResult=true, interpSpeedDecreasing=1.7, interpSpeedIncreasing=4}
    boneToModify {boneName=base_flap_l}
    rotation {pitch=-10, yaw=-20}
   }
   n11 = Get Jumpfix() -> (Jumpfix:double)
   n12 = "Transform (Modify) Bone - Bone: dyn_flap_r"(ComponentPose=@n10, Alpha=@n11, Translation="0,0,0", Rotation="-15,-10,0", Scale="1,1,1") {
    rotationMode=BMM_Additive
    rotationSpace=BCS_BoneSpace
    translationMode=BMM_Additive
    translationSpace=BCS_BoneSpace
    alphaBoolBlend {blendOption=Linear}
    alphaScaleBiasClamp {bInterpResult=true, interpSpeedDecreasing=6, interpSpeedIncreasing=0.5, scale=5}
    boneToModify {boneName=dyn_flap_r}
    rotation {pitch=-15, yaw=-10}
   }
   n13 = "Transform (Modify) Bone - Bone: dyn_flap_l"(ComponentPose=@n12, Alpha=@n11, Translation="0,0,0", Rotation="-15,0,0", Scale="1,1,1") {
    rotationMode=BMM_Additive
    rotationSpace=BCS_BoneSpace
    translationMode=BMM_Additive
    translationSpace=BCS_BoneSpace
    alphaBoolBlend {blendOption=Linear}
    alphaScaleBiasClamp {bInterpResult=true, interpSpeedDecreasing=6, interpSpeedIncreasing=0.5, scale=5}
    boneToModify {boneName=dyn_flap_l}
    rotation {pitch=-15}
   }
   n14 = "Transform (Modify) Bone - Bone: dyn_flap_01_r"(ComponentPose=@n13, Alpha=@n11, Translation="0,0,0", Rotation="-5,0,0", Scale="1,1,1") {
    rotationMode=BMM_Additive
    rotationSpace=BCS_BoneSpace
    translationMode=BMM_Additive
    translationSpace=BCS_BoneSpace
    alphaBoolBlend {blendOption=Linear}
    alphaScaleBiasClamp {bInterpResult=true, interpSpeedDecreasing=6, interpSpeedIncreasing=0.5, scale=5}
    boneToModify {boneName=dyn_flap_01_r}
    rotation {pitch=-5}
   }
   n15 = "Transform (Modify) Bone - Bone: dyn_flap_01_l"(ComponentPose=@n14, Alpha=@n11, Translation="0,0,0", Rotation="-5,0,0", Scale="1,1,1") {
    rotationMode=BMM_Additive
    rotationSpace=BCS_BoneSpace
    translationMode=BMM_Additive
    translationSpace=BCS_BoneSpace
    alphaBoolBlend {blendOption=Linear}
    alphaScaleBiasClamp {bInterpResult=true, interpSpeedDecreasing=6, interpSpeedIncreasing=0.5, scale=5}
    boneToModify {boneName=dyn_flap_01_l}
    rotation {pitch=-5}
   }
   n16 = "Transform (Modify) Bone - Bone: dyn_flap_02_r"(ComponentPose=@n15, Alpha=@n11, Translation="0,0,0", Rotation="-25,0,0", Scale="1,1,1") {
    rotationMode=BMM_Additive
    rotationSpace=BCS_BoneSpace
    translationMode=BMM_Additive
    translationSpace=BCS_BoneSpace
    alphaBoolBlend {blendOption=Linear}
    alphaScaleBiasClamp {bInterpResult=true, interpSpeedDecreasing=6, interpSpeedIncreasing=0.5, scale=5}
    boneToModify {boneName=dyn_flap_02_r}
    rotation {pitch=-25}
   }
   n17 = "Transform (Modify) Bone - Bone: dyn_flap_02_l"(ComponentPose=@n16, Alpha=@n11, Translation="0,0,0", Rotation="-25,0,0", Scale="1,1,1") {
    rotationMode=BMM_Additive
    rotationSpace=BCS_BoneSpace
    translationMode=BMM_Additive
    translationSpace=BCS_BoneSpace
    alphaBoolBlend {blendOption=Linear}
    alphaScaleBiasClamp {bInterpResult=true, interpSpeedDecreasing=6, interpSpeedIncreasing=0.5, scale=5}
    boneToModify {boneName=dyn_flap_02_l}
    rotation {pitch=-25}
   }
   Output Pose(Result=Component To Local(ComponentPose=@n17)) {layerGroup=None}
   Save cached pose 'inputNode'(Pose=Input Pose()) {cacheName=inputNode}
  ]}
 EventGraph {nodes [
   Event BlueprintUpdateAnimation()
   > n1 = Sequence!(execute) -> (then_0, then_1, then_2:Exec)
   > n2 = Gravity Override State Function DEPRECATED(CharacterPartAnimInstance=Self-Reference(), GameTimeInSeconds=GetGameTimeInSeconds(), GravityOverrideParams=Get Gravity Override Params())
   > Set GravityOverride(GravityOverride=@n2)
   n3 = Branch(execute=@n1.then_1, Condition=Get bIsSkyDiving())
   > Set SkyFix(SkyFix:double=Get InterpAlpha() -> (InterpAlpha:double)) -> (Output_Get:double)
   Set SkyFix(execute=@n3.else, SkyFix=Get RegAlpha() -> (RegAlpha:double))
   n4 = Branch(execute=@n1.then_2, Condition=Get bIsJumping())
   > Delay(Duration=0.3)
   > Set Jumpfix(Jumpfix:double=Get InterpAlpha()) -> (Output_Get:double)
   Delay(execute=@n4.else, Duration=0.3)
   > Set Jumpfix(Jumpfix=Get RegAlpha())
   TryGetPawnOwner()
  ]}
]
variables [
 Gravity Override Params {
  type=GravityOverrideParamsStruct
  default {
   skydive_FloatingDown {gravityOverride=<0 -700 -4000>}
   skydive_GoingUp {gravityOverride=<0 -300 -4000>}
   skydive_Parachuting {gravityOverride=<0 -700 -2000>}
  }
 }
 GravityOverride {type=Vector}
 InterpAlpha {type=number, default=1}
 Jumpfix {type=number}
 RegAlpha {type=number}
 SkyFix {type=number}
]
